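# Animation Blueprint for the FirePetal break-action shotgun's weapon mesh, animating its two shots and reloads.
# FireSM moves between idle, shot, spent-pose and reload states using AmmoRemaining, bIsReloading and automatic sequence-end rules, mostly without crossfade.
# Reloads can interrupt shots and exit early when reloading stops; output passes through DefaultSlot for montages.
# BeginPlay computes ReloadPlayRate from the reload montage's length divided by the weapon stats row's ReloadTime.
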AnimBlueprint /FirePetalWeaponGameplay/Gameplay/FirePetal_FlavorShotgun/WeaponAnimation/FirePetal_Break_Shotgun_AnimBP
target_skeleton=/FirePetalWeaponGameplay/Gameplay/FirePetal_FlavorShotgun/Meshes/Break_FirePetal_Shotgun_Skeleton
graphs [
 AnimGraph {nodes [
   Local Space Ref Pose()
   AnimGraphNode_StateMachine_0 = FireSM() {is=title AnimGraphNode_StateMachine}
   Output Pose(Result=Slot 'DefaultSlot'(Source=@AnimGraphNode_StateMachine_0))
  ]}
 FireSM {
  entry=Idle
  path=AnimGraph.AnimGraphNode_StateMachine_0.FireSM
  ids=id
  nodes [
   AnimStateNode_0 = Idle()
   AnimStateNode_1 = FirstShot() {is=title AnimStateNode}
   AnimStateNode_2 = FirstShotIdle() {is=title AnimStateNode}
   AnimStateNode_3 = SecondShot() {is=title AnimStateNode}
   AnimStateNode_4 = SecondShotIdle() {is=title AnimStateNode}
   AnimStateNode_5 = Reload1() {is=title AnimStateNode}
   AnimStateNode_6 = Reload2() {is=title AnimStateNode}
   AnimStateTransitionNode_0 = AnimStateTransitionNode() {crossfadeDuration=0}
   AnimStateTransitionNode_1 = AnimStateTransitionNode() {bAutomaticRuleBasedOnSequencePlayerInState=true, crossfadeDuration=0}
   AnimStateTransitionNode_10 = AnimStateTransitionNode() {crossfadeDuration=0}
   AnimStateTransitionNode_11 = AnimStateTransitionNode() {crossfadeDuration=0}
   AnimStateTransitionNode_12 = AnimStateTransitionNode() {crossfadeDuration=0}
   AnimStateTransitionNode_13 = AnimStateTransitionNode()
   AnimStateTransitionNode_14 = AnimStateTransitionNode()
   AnimStateTransitionNode_2 = AnimStateTransitionNode() {crossfadeDuration=0}
   AnimStateTransitionNode_3 = AnimStateTransitionNode() {bAutomaticRuleBasedOnSequencePlayerInState=true, crossfadeDuration=0}
   AnimStateTransitionNode_4 = AnimStateTransitionNode()
   AnimStateTransitionNode_5 = AnimStateTransitionNode()
   AnimStateTransitionNode_6 = AnimStateTransitionNode() {crossfadeDuration=0}
   AnimStateTransitionNode_7 = AnimStateTransitionNode() {bAutomaticRuleBasedOnSequencePlayerInState=true, crossfadeDuration=0}
   AnimStateTransitionNode_8 = AnimStateTransitionNode() {crossfadeDuration=0}
   AnimStateTransitionNode_9 = AnimStateTransitionNode() {bAutomaticRuleBasedOnSequencePlayerInState=true, crossfadeDuration=0}
  ]
  links [
   {
    dst=FirstShotIdle
    rule=AnimGraph.AnimGraphNode_StateMachine_0.FireSM.AnimStateTransitionNode_1.Transition
    src=FirstShot
    via=AnimStateTransitionNode_1
   }
   {
    dst=Reload1
    rule=AnimGraph.AnimGraphNode_StateMachine_0.FireSM.AnimStateTransitionNode_10.Transition
    src=FirstShot
    via=AnimStateTransitionNode_10
   }
   {
    dst=Idle
    rule=AnimGraph.AnimGraphNode_StateMachine_0.FireSM.AnimStateTransitionNode_4.Transition
    src=FirstShotIdle
    via=AnimStateTransitionNode_4
   }
   {
    dst=Reload1
    rule=AnimGraph.AnimGraphNode_StateMachine_0.FireSM.AnimStateTransitionNode_6.Transition
    src=FirstShotIdle
    via=AnimStateTransitionNode_6
   }
   {
    dst=SecondShot
    rule=AnimGraph.AnimGraphNode_StateMachine_0.FireSM.AnimStateTransitionNode_2.Transition
    src=FirstShotIdle
    via=AnimStateTransitionNode_2
   }
   {
    dst=FirstShot
    rule=AnimGraph.AnimGraphNode_StateMachine_0.FireSM.AnimStateTransitionNode_0.Transition
    src=Idle
    via=AnimStateTransitionNode_0
   }
   {
    dst=Reload2
    rule=AnimGraph.AnimGraphNode_StateMachine_0.FireSM.AnimStateTransitionNode_12.Transition
    src=Idle
    via=AnimStateTransitionNode_12
   }
   {
    dst=FirstShotIdle
    rule=AnimGraph.AnimGraphNode_StateMachine_0.FireSM.AnimStateTransitionNode_14.Transition
    src=Reload1
    via=AnimStateTransitionNode_14
   }
   {
    dst=Idle
    rule=AnimGraph.AnimGraphNode_StateMachine_0.FireSM.AnimStateTransitionNode_7.Transition
    src=Reload1
    via=AnimStateTransitionNode_7
   }
   {
    dst=Idle
    rule=AnimGraph.AnimGraphNode_StateMachine_0.FireSM.AnimStateTransitionNode_9.Transition
    src=Reload2
    via=AnimStateTransitionNode_9
   }
   {
    dst=SecondShotIdle
    rule=AnimGraph.AnimGraphNode_StateMachine_0.FireSM.AnimStateTransitionNode_13.Transition
    src=Reload2
    via=AnimStateTransitionNode_13
   }
   {
    dst=Reload2
    rule=AnimGraph.AnimGraphNode_StateMachine_0.FireSM.AnimStateTransitionNode_11.Transition
    src=SecondShot
    via=AnimStateTransitionNode_11
   }
   {
    dst=SecondShotIdle
    rule=AnimGraph.AnimGraphNode_StateMachine_0.FireSM.AnimStateTransitionNode_3.Transition
    src=SecondShot
    via=AnimStateTransitionNode_3
   }
   {
    dst=Idle
    rule=AnimGraph.AnimGraphNode_StateMachine_0.FireSM.AnimStateTransitionNode_5.Transition
    src=SecondShotIdle
    via=AnimStateTransitionNode_5
   }
   {
    dst=Reload2
    rule=AnimGraph.AnimGraphNode_StateMachine_0.FireSM.AnimStateTransitionNode_8.Transition
    src=SecondShotIdle
    via=AnimStateTransitionNode_8
   }
  ]
 }
 Idle {
  path=AnimGraph.AnimGraphNode_StateMachine_0.FireSM.AnimStateNode_0.Idle
  nodes [Output Animation Pose(Result=Sequence Player 'FirePetal_Break_Shotgun_Pose_Default_WEP'() {sequence=&/FirePetal_Break_Shotgun_Pose_Default_WEP})]
 }
 FirstShot {
  path=AnimGraph.AnimGraphNode_StateMachine_0.FireSM.AnimStateNode_1.FirstShot
  nodes [
   n1 = Sequence Player 'FirePetal_Break_Shotgun_Fire_1_WEP'() {bLoopAnimation=false, sequence=&/FirePetal_Break_Shotgun_Fire_1_WEP}
   Output Animation Pose(Result=@n1)
  ]
 }
 FirstShotIdle {
  path=AnimGraph.AnimGraphNode_StateMachine_0.FireSM.AnimStateNode_2.FirstShotIdle
  nodes [Output Animation Pose(Result=Sequence Player 'FirePetal_Break_Shotgun_Pose_Fire_First_WEP'() {sequence=&/FirePetal_Break_Shotgun_Pose_Fire_First_WEP})]
 }
 SecondShot {
  path=AnimGraph.AnimGraphNode_StateMachine_0.FireSM.AnimStateNode_3.SecondShot
  nodes [
   n1 = Sequence Player 'FirePetal_Break_Shotgun_Fire_2_WEP'() {bLoopAnimation=false, sequence=&/FirePetal_Break_Shotgun_Fire_2_WEP}
   Output Animation Pose(Result=@n1)
  ]
 }
 SecondShotIdle {
  path=AnimGraph.AnimGraphNode_StateMachine_0.FireSM.AnimStateNode_4.SecondShotIdle
  nodes [Output Animation Pose(Result=Sequence Player 'FirePetal_Break_Shotgun_Pose_Fire_Second_WEP'() {sequence=&/FirePetal_Break_Shotgun_Pose_Fire_Second_WEP})]
 }
 Reload1 {
  path=AnimGraph.AnimGraphNode_StateMachine_0.FireSM.AnimStateNode_5.Reload1
  nodes [
   n1 = Sequence Player 'FirePetal_Break_Shotgun_Reload_1WEP'!(PlayRate:float=1) -> (Pose) {bLoopAnimation=false, sequence=&/FirePetal_Break_Shotgun_Reload_1WEP}
   Output Animation Pose(Result=@n1)
  ]
 }
 Reload2 {
  path=AnimGraph.AnimGraphNode_StateMachine_0.FireSM.AnimStateNode_6.Reload2
  nodes [
   n1 = Sequence Player 'FirePetal_Break_Shotgun_Reload_2WEP'!(PlayRate:float=1) -> (Pose) {bLoopAnimation=false, sequence=&/FirePetal_Break_Shotgun_Reload_2WEP}
   Output Animation Pose(Result=@n1)
  ]
 }
 Transition {
  path=AnimGraph.AnimGraphNode_StateMachine_0.FireSM.AnimStateTransitionNode_0.Transition
  nodes [Result(bCanEnterTransition="Equal (Integer)"(A=Get AmmoRemaining() -> (AmmoRemaining:int=0), B=1))]
 }
 Transition {path=AnimGraph.AnimGraphNode_StateMachine_0.FireSM.AnimStateTransitionNode_1.Transition, nodes [Result()]}
 Transition {
  path=AnimGraph.AnimGraphNode_StateMachine_0.FireSM.AnimStateTransitionNode_10.Transition
  nodes [
   n1 = Get AmmoRemaining() -> (AmmoRemaining=0)
   Result(bCanEnterTransition=AND Boolean(A=Get bIsReloading() -> (bIsReloading:bool), B="Equal (Integer)"(A=@n1, B=1)))
  ]
 }
 Transition {
  path=AnimGraph.AnimGraphNode_StateMachine_0.FireSM.AnimStateTransitionNode_11.Transition
  nodes [
   n1 = Get AmmoRemaining() -> (AmmoRemaining=0)
   Result(bCanEnterTransition=AND Boolean(A=Get bIsReloading(), B="Equal (Integer)"(A=@n1)))
  ]
 }
 Transition {
  path=AnimGraph.AnimGraphNode_StateMachine_0.FireSM.AnimStateTransitionNode_12.Transition
  nodes [
   n1 = Get AmmoRemaining() -> (AmmoRemaining=0)
   Result(bCanEnterTransition=AND Boolean(A=Get bIsReloading(), B="Equal (Integer)"(A=@n1)))
  ]
 }
 Transition {
  path=AnimGraph.AnimGraphNode_StateMachine_0.FireSM.AnimStateTransitionNode_13.Transition
  nodes [Result(bCanEnterTransition=NOT Boolean(A=Get bIsReloading()))]
 }
 Transition {
  path=AnimGraph.AnimGraphNode_StateMachine_0.FireSM.AnimStateTransitionNode_14.Transition
  nodes [Result(bCanEnterTransition=NOT Boolean(A=Get bIsReloading()))]
 }
 Transition {
  path=AnimGraph.AnimGraphNode_StateMachine_0.FireSM.AnimStateTransitionNode_2.Transition
  nodes [Result(bCanEnterTransition="Equal (Integer)"(A=Get AmmoRemaining() -> (AmmoRemaining=0)))]
 }
 Transition {path=AnimGraph.AnimGraphNode_StateMachine_0.FireSM.AnimStateTransitionNode_3.Transition, nodes [Result()]}
 Transition {
  path=AnimGraph.AnimGraphNode_StateMachine_0.FireSM.AnimStateTransitionNode_4.Transition
  nodes [Result(bCanEnterTransition="Equal (Integer)"(A=Get AmmoRemaining() -> (AmmoRemaining=0), B=2))]
 }
 Transition {
  path=AnimGraph.AnimGraphNode_StateMachine_0.FireSM.AnimStateTransitionNode_5.Transition
  nodes [Result(bCanEnterTransition="Equal (Integer)"(A=Get AmmoRemaining() -> (AmmoRemaining=0), B=2))]
 }
 Transition {
  path=AnimGraph.AnimGraphNode_StateMachine_0.FireSM.AnimStateTransitionNode_6.Transition
  nodes [
   n1 = Get AmmoRemaining() -> (AmmoRemaining=0)
   Result(bCanEnterTransition=AND Boolean(A=Get bIsReloading(), B="Equal (Integer)"(A=@n1, B=1)))
  ]
 }
 Transition {path=AnimGraph.AnimGraphNode_StateMachine_0.FireSM.AnimStateTransitionNode_7.Transition, nodes [Result()]}
 Transition {
  path=AnimGraph.AnimGraphNode_StateMachine_0.FireSM.AnimStateTransitionNode_8.Transition
  nodes [
   n1 = Get AmmoRemaining() -> (AmmoRemaining=0)
   Result(bCanEnterTransition=AND Boolean(A=Get bIsReloading(), B="Equal (Integer)"(A=@n1)))
  ]
 }
 Transition {path=AnimGraph.AnimGraphNode_StateMachine_0.FireSM.AnimStateTransitionNode_9.Transition, nodes [Result()]}
 EventGraph {nodes [
   Event BlueprintBeginPlay() {is=title K2Node_Event}
   > n1 = Cast To B_Shotgun_FirePetal_Flavor_Athena!(execute, Object=GetOwningActor()) -> (then, CastFailed, AsB Shotgun Fire Petal Flavor Athena:B Shotgun Fire Petal Flavor Athena Object Reference) {is=title K2Node_DynamicCast}
   n2 = Break Data Table Row Handle!(DataTableRowHandle:"Data Table Row Handle Structure (by ref)"=GetWeaponDataWeaponStatHandle!(self:Fort Weapon Object Reference=@n1) -> (ReturnValue:Data Table Row Handle Structure)) -> (DataTable:Data Table Object Reference, RowName:Name) {is=title K2Node_BreakStruct}
   > n3 = Get Data Table Row!(execute:Exec, DataTable:Data Table Object Reference=@n2, RowName:Name=@n2.RowName) -> (then:Exec, RowNotFound:Exec, ReturnValue:Fort Ranged Weapon Stats Structure)
   n4 = GetReloadAnimation!(self:Fort Weapon Object Reference=@n1) -> (ReturnValue:Anim Montage Object Reference)
   > Set ReloadPlayRate(ReloadPlayRate:double=float / float(
     A=Get SequenceLength!(self:Anim Sequence Base Object Reference=@n4) -> (SequenceLength:float)
     B=Break Fort Ranged Weapon Stats!(FortRangedWeaponStats:"Fort Ranged Weapon Stats Structure (by ref)"=@n3) -> (
      JumpingFallingSpreadDownsights:float=~
      bUseConstantCooldownRecoverySpeed:bool=~
      ConstantCooldownRecoverySpeed:float=~
      EarlyRefireRate:float=~
      EarlyRefireSpreadPenalty:float=~
      EarlyRefireSpreadMinimumMultiplier:float=~
      EarlyRefireSpreadDownsightsMinimumMultiplier:float=~
      RecoilRecoveryEaseInTime:float=~
      RecoilStandingStillMultiplier:float=~
      RecoilMovingMultiplier:float=~
      bEvenRecoilWhenCountered:bool=~
      OffsetRecoilFactor:float=~
      RecoilBounceFactor:float=~
      RecoilBounceOffsetFactor:float=~
      OffsetRecoilLimitOverride:float=~
      AIAthenaHearLightweightProjectileImpactNoiseRange:Scalable Float Structure
      AIAthenaHearLightweightProjectileFlyByNoiseRange:Scalable Float Structure
      AIAthenaLightweightProjectileFlyByEmissionInterval:Scalable Float Structure
      bCoolOverheatWhileUnequipped:bool=~
      ManualVentingMinOverheatValueRequirement:float=~
      bUseConstantLerpForRecoil:bool=~
      bStartRecoilDelayAfterRecoilPeaks:bool=~
      bUseAccumulatedRecoilToSelectShotIndex:bool=~
      RecoilFractionToBulletCountCurve:Curve Float Soft Object Reference
      bUseAccumulatedRecoilToLerpMinToMaxPerBulletRecoilValues:bool=~
      RecoilMinToMaxCurve:Curve Float Soft Object Reference
      RecoilMinVerticalCurve:Curve Float Soft Object Reference
      RecoilMaxVerticalCurve:Curve Float Soft Object Reference
      RecoilMinHorizontalCurve:Curve Float Soft Object Reference
      RecoilMaxHorizontalCurve:Curve Float Soft Object Reference
      RecoilErrorAddedPerShot:float=~
      RecoilOffsetConsideredMaxFraction:float=~
      RecoilDelay:float=~
      bDelaySpreadWithRecoil:bool=~
      bDisableOneFrameFiringAnimationDelay:bool=~
      bUseScopedPerBulletRecoilOverridesForBasicADS:bool=~
      bLerpSprintingSpreadMultiplierWithMovementSpeed:bool=~
      SpreadMultiplierCap:float=~
      ReloadTime:float=~
      StowAnimRate:float=~
      AimAssistAutoAimSize:float=~
      AutofireMinRange:float=~
      AutofireZoomTargetMinSizePercent:float=~
      AutofireMinChargingTime:float=~
      PerWeaponHotfixes:Fort Per Weapon Hotfixes Structure
     ) {is=title K2Node_BreakStruct}.ReloadTime
    )) -> (Output_Get:double)
  ]}
]
variables [ReloadPlayRate {type=number, default=1}]
